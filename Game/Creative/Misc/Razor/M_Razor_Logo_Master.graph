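# Master material for the Creative Razor logo: a lit metallic surface with an optional animated emissive, gated by the CrazyON switch and scaled by Fader.
# Stepped V coordinates form scanline bands, lit by a looping time sweep and a timestamp in/out transition.
# The transition blends two Fresnel-graded color palettes, with the sweep brightening the second.
# An LCD function modulates the glow, and the attributes feed MF_CreativeMoveTool.
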Material /Game/Creative/Misc/Razor/M_Razor_Logo_Master
graph {
 properties {
  bUseMaterialAttributes=true
  bUsedWithMeshDeformer=true
  bUsedWithSkeletalMesh=true
  pixelDepthOffsetMode=PDOM_Legacy
  usedShadingModels=Default Lit
 }
 nodes [
  n1 = ScalarParameter() {defaultValue=25, parameterName=Steps}
  n2 = Multiply(A=ComponentMask(None=TextureCoordinate()) {g=true}, B=@n1)
  n3 = Frac(None=@n2)
  n4 = OneMinus(None=@n3)
  n5 = Multiply(
   A=Add(A=Multiply(A=@n4, B=@n3), B=ScalarParameter() {defaultValue=-0.04, parameterName=Line Bias})
   B=ScalarParameter() {defaultValue=8.5, parameterName=Line Scale}
  )
  n6 = OneMinus(None=Divide(A=Ceil(None=@n2), B=@n1))
  n7 = Add(A=Time() {period=1}, B=ScalarParameter() {parameterName=TimeStamp Offset})
  n8 = ScalarParameter() {defaultValue=0.75, parameterName=Speed A}
  n9 = Multiply(A=Frac(None=Multiply(A=@n7, B=@n8))) {constB=2}
  n10 = Constant() {r=1}
  n11 = ScalarParameter() {defaultValue=25, parameterName=Steps}
  n12 = Multiply(A=@n9, B=@n11)
  n13 = OneMinus(None=Divide(A=Ceil(None=@n12), B=@n11))
  n14 = Multiply(A=If(A=@n6, B=@n9, "A > B"=Constant(), "A == B"=@n10, "A < B"=@n10), B=Clamp(None=Add(A=@n13, B=@n6)))
  n15 = Constant()
  n16 = If(
   A=@n6
   B=MaterialFunctionCall(Transition In Only?=StaticBool()) {materialFunction=&MF_Timestamp_InOut_01}.Linear
   "A > B"=@n15
   "A == B"=@n15
   "A < B"=Constant() {r=1}
  )
  n17 = Power(Base=Clamp(None=Add(A=@n14, B=@n16))) {constExponent=10}
  n18 = MaterialFunctionCall(Exponent=ScalarParameter() {defaultValue=5, parameterName=Fresnel Exponent}) {materialFunction=&MF_Fresnel_FX}
  n19 = Multiply(A=Power(Base=@n14) {constExponent=20}, B=ScalarParameter() {defaultValue=4, parameterName=Sweep Brightness})
  n20 = LinearInterpolate(
   A=LinearInterpolate(
    A=VectorParameter() {parameterName=Emissive A Outer, defaultValue=<rgba 0.6 0.15 0 1>}.RGB
    B=VectorParameter() {parameterName=Emissive A Inner, defaultValue=<rgba 8 7 4 1>}.RGB
    Alpha=@n18.Fresnel
   )
   B=LinearInterpolate(
    A=VectorParameter() {parameterName=Emissive B Outer, defaultValue=<rgba 0 0.06 0.3 1>}.RGB
    B=VectorParameter() {parameterName=Emissive B Inner, defaultValue=<rgba 0.15 1.12 5 1>}.RGB
    Alpha=Add(A=@n19, B=@n18.Fresnel)
   )
   Alpha=@n16
  )
  n21 = Multiply(
   A=Multiply(A=Clamp(None=@n5), B=Multiply(A=@n17, B=@n20))
   B=MaterialFunctionCall!(U Repeat, V Repeat, Dark Value, Light Value) {materialFunction=/Game/Effects/Fort_Effects/Effects/Atlas/Materials/MF_Atlas_LCD_01}.Result
  )
  n22 = ScalarParameter() {parameterName=Fader}
  n23 = SetMaterialAttributes!(
   MaterialAttributes
   Base Color=VectorParameter() {parameterName=Base Color, defaultValue=<rgba 0.1297 0.1131 0.19 1>}.RGB
   Metallic=ScalarParameter() {defaultValue=1, parameterName=Metallic}
   Specular=ScalarParameter() {parameterName=Specular}
   Roughness=ScalarParameter() {parameterName=Roughness}
   Emissive Color=StaticSwitch(True=Multiply(A=@n21, B=@n22), False=Constant(), Value=StaticBoolParameter() {parameterName=CrazyON})
  ) {attributeSetTypes [
    #1
    #2
    #3
    #4
    #5
   ]}
  MaterialFunctionCall(MA In=@n23) {materialFunction=&MF_CreativeMoveTool}
  ScalarParameter() {defaultValue=3, parameterName=Fake Time}
  ScalarParameter() {defaultValue=1, parameterName=A}
  ScalarParameter() {defaultValue=2, parameterName=B}
  ScalarParameter() {defaultValue=2, parameterName=In}
  ScalarParameter() {defaultValue=2, parameterName=Out}
 ]
}
